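; Interactable lemon cart button for Fort pawns that plays a stinger sound and a press animation when used.
; It is always interactable, takes its prompt text from a ButtonString variable and overrides the interaction distance.
; Interacting flips a replicated RepTrigger boolean, whose OnRep plays the stinger audio and a debug print where cosmetic events can run.
; The OnRep also runs ButtonAnimation, a timeline moving the button mesh and driving glow light intensity.
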(event EventBeginPlay)

(event Collision|EventActorBeginOverlap
  (params (OtherActor :type Actor%20Object%20Reference)))

(event EventTick
  (params (DeltaSeconds :type Float%20%28single-precision%29)))

(event Interaction|EventBlueprintOnInteract
  (params (InteractingPawn :type Fort%20Pawn%20Object%20Reference) (InteractionBeingAttempted :type EInteractionBeingAttempted%20Enum :default "FirstInteraction"))
  ;[1 1 1] Manual Interaction
  (Variables|Default|SetRep%20Trigger (not (Variables|Default|GetRep%20Trigger))))

(event Custom|ButtonAnimation
  ; Button Animation
  (bind (_direction _buttonvector _glow) (Timeline|ButtonMovement PlayFromStart)
    (:Update
      (Transformation|SetRelativeLocation (Variables|Default|Getbuttonmesh) _buttonvector false false)
      (Rendering|Components|Light|SetIntensity (Variables|Default|GetButtonGlowLight) _glow))))

(fn BlueprintCanInteract
  (params (InteractingPawn :type Fort%20Pawn%20Object%20Reference) (InteractionBeingAttempted :type EInteractionBeingAttempted%20Enum :default "FirstInteraction") (InteractionType :type TInteractionType%20Enum :default "IT_NoInteraction"))
  (return true))

(fn BlueprintGetInteractionString
  (params (InteractingPawn :type Fort%20Pawn%20Object%20Reference) (InteractionBeingAttempted :type EInteractionBeingAttempted%20Enum :default "FirstInteraction"))
  (return (Utilities|Text|ToText(String) (Utilities|String|ToString(Name) (Variables|Default|GetButtonString)))))

(fn GetMinDistanceFromInteraction
  (return 4e+02))

(fn OnRep_RepTrigger
  (bind _stinger (Variables|Default|GetStinger))
  (Networking|CanExecuteCosmeticEvents
    (:True
      (Utilities|FlowControl|Sequence
        (:then_0
          (Components|Activation|Activate _stinger false)
          (Audio|Components|Audio|Play _stinger 0.0)
          (Development|PrintString "Playing Stinger Audio" true true "(R=1.000000,G=0.335960,B=0.000000,A=1.000000)" 1e+01))
        (:then_1
          (CallFunction|ButtonAnimation))))))

(fn ConstructionScript)
