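# Material function that renders animated LemonCart screens and blends them into input material attributes by a mask.
# A Switch picks one of four subscreens by index, position-based randomness or timed cycling, then feeds a transition function.
# Sine-driven jitter, flicker and scanlines, plus a ramp-texture bump offset, add screen noise.
# View-angle darkening shapes color and emissive, with ray-tracing quality switches for Lumen boost and eye-adaptation compensation.
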MaterialFunction /Clyde/Environments/Sets/LemonCart/MaterialFunctions/MF_LemonCartScreens
graph {nodes [
  n1 = NamedRerouteUsage()
  n2 = NamedRerouteUsage()
  n3 = MaterialFunctionCall!(UVs=NamedRerouteUsage(), Height Bump Offset=NamedRerouteUsage(), Vertical Jitter=Constant(), Seed=NamedRerouteUsage()) {materialFunction=&/MF_LemonCart_Subscreen_01}
  MaterialExpressionReroute_0 = Reroute(None=AppendVector(A=@n3.Result, B=@n3.Result))
  n4 = MaterialFunctionCall!(UVs=NamedRerouteUsage(), Speed, Mask=NamedRerouteUsage()) {materialFunction=&/MF_LemonCart_Subscreen_02}
  MaterialExpressionAppendVector_2 = AppendVector(A=@n4.Result, B=@n4.Result)
  n5 = Add(
   A=MaterialFunctionCall(UVs=NamedRerouteUsage(), Texture Scale=Constant() {r=1.3}) {materialFunction=&ScaleUVsByCenter}.UVs
   B=NamedRerouteUsage()
  )
  n6 = Constant2Vector() {g=0.1}
  n7 = TextureSample!(UVs=BumpOffset(Coordinate=Add(A=@n5, B=@n6), Height=NamedRerouteUsage()), Tex, Level, Apply View MipBias) {
   desc="Dog image is always looking low quiality for some reason\r\n- forced MIP to ZERO"
   mipValueMode=TMVM_MipLevel
   texture=/Clyde/Environments/Sets/LemonCart/Textures/T_LemonCart_Screen_01
  }
  n8 = Multiply(
   A=Power(Base=@n7.B, Exp=ScalarParameter() {defaultValue=0.5, parameterName=SubScreen03.Contrast})
   B=ScalarParameter() {defaultValue=0.6, parameterName=SubScreen03.Brightness}
  ) {constB=0.6}
  n9 = Multiply(A=@n3.Result, B=@n8)
  MaterialExpressionLinearInterpolate_1 = LinearInterpolate(A=AppendVector(A=@n9, B=@n8), B=AppendVector(A=@n9, B=Ceil(None=@n7.B)), Alpha=Constant() {r=0.8})
  n10 = MaterialFunctionCall!(UVs=NamedRerouteUsage(), Height Bump Offset=NamedRerouteUsage(), Vertical Jitter=NamedRerouteUsage()) {materialFunction=&/MF_LemonCart_Subscreen_03}
  MaterialExpressionReroute_1 = Reroute(None=AppendVector(A=@n10.Result, B=@n10.Result))
  n11 = ScalarParameter() {defaultValue=4, group=ScreenControl, parameterName=ScreenControl.NumberOfScreens}
  n12 = Round(None=Multiply(A=NamedRerouteUsage(), B=Subtract(A=@n11)) {constB=3})
  n13 = LinearInterpolate(
   A=ScalarParameter() {defaultValue=5, desc=High number is slower, group=ScreenControl, parameterName=ScreenControl.CycleSpeedMin}
   B=ScalarParameter() {defaultValue=5, desc=High number is slower, group=ScreenControl, parameterName=ScreenControl.CycleSpeedMax}
   Alpha=NamedRerouteUsage()
  )
  n14 = Divide(A=Add(A=Time(), B=NamedRerouteUsage()), B=@n13) {constB=5}
  n15 = Modulo!(
   A=Add(A=ScalarParameter() {group=ScreenControl, parameterName=ScreenControl.Index}, B=Round(None=@n14))
   B=ScalarParameter() {defaultValue=4, group=ScreenControl, parameterName=ScreenControl.NumberOfScreens}
  )
  n16 = ScalarParameter() {defaultValue=1, desc=Must use random screens, group=ScreenControl, parameterName=ScreenControl.CycleScreens}
  n17 = Switch!(
   SwitchValue=LinearInterpolate(
    A=ScalarParameter() {group=ScreenControl, parameterName=ScreenControl.Index}
    B=LinearInterpolate(A=@n12, B=@n15, Alpha=@n16)
    Alpha=ScalarParameter() {defaultValue=1, group=ScreenControl, parameterName=ScreenControl.UseRandomScreen}
   )
   Default=@MaterialExpressionReroute_0
   Screen1=@MaterialExpressionReroute_0
   Screen2=@MaterialExpressionAppendVector_2
   Screen3=@MaterialExpressionLinearInterpolate_1
   Screen4=@MaterialExpressionReroute_1
  ) {inputs [
    {
     inputName=Screen1
     input {
      expression=&self:MaterialExpressionReroute_0
      inputName=None
      mask=0
      maskA=0
      maskB=0
      maskG=0
      maskR=0
      outputIndex=0
     }
    }
    {
     inputName=Screen2
     input {
      expression=&self:MaterialExpressionAppendVector_2
      inputName=None
      mask=0
      maskA=0
      maskB=0
      maskG=0
      maskR=0
      outputIndex=0
     }
    }
    {
     inputName=Screen3
     input {
      expression=&self:MaterialExpressionLinearInterpolate_1
      inputName=None
      mask=0
      maskA=0
      maskB=0
      maskG=0
      maskR=0
      outputIndex=0
     }
    }
    {
     inputName=Screen4
     input {
      expression=&self:MaterialExpressionReroute_1
      inputName=None
      mask=0
      maskA=0
      maskB=0
      maskG=0
      maskR=0
      outputIndex=0
     }
    }
   ]}
  n18 = MaterialFunctionCall!(UVs=NamedRerouteUsage(), TransitionLength=@n2, HoldLength, Logo RGBA=@n17, TimeOffset=Divide(A=@n2) {constB=-2}) {materialFunction=&/MF_LemonCart_Transition}
  n19 = NamedRerouteUsage()
  n20 = VectorParameter() {parameterName=Color.Base, defaultValue=<rgba 0.03 0.0285 0.0256 1>}
  n21 = Add(A=NamedRerouteUsage(), B=LinearInterpolate(A=Multiply(A=@n1, B=@n18.Result), B=Add(A=@n19, B=@n20.RGB), Alpha=NamedRerouteUsage()))
  n22 = CameraVectorWS()
  n23 = VertexNormalWS()
  n24 = Multiply(A=Power(Base=DotProduct(A=@n22, B=@n23), Exp=Constant() {r=2})) {constB=2}
  n25 = Multiply(A=@n21, B=LinearInterpolate(A=ScalarParameter() {parameterName=DarkeningEffect.Min}, Alpha=Saturate(None=@n24)))
  n26 = Constant() {r=0.15}
  n27 = ScalarParameter() {defaultValue=1, parameterName=LumenBoost}
  n28 = Multiply(
   A=RayTracingQualitySwitch(Normal=@n25, RayTraced=Multiply(A=@n21, B=@n27))
   B=ScalarParameter() {defaultValue=1, group=ScreenControl, parameterName=EmissiveMulti}
  )
  n29 = LinearInterpolate(
   A=@n28
   B=Divide(A=@n28, B=EyeAdaptation().EyeAdaptation)
   Alpha=RayTracingQualitySwitch(Normal=ScalarParameter() {defaultValue=1, parameterName=UseEyeAdaptation}, RayTraced=Constant() {r=1})
  )
  n30 = Constant() {r=1}
  FunctionOutput(None=BlendMaterialAttributes(
    A=FunctionInput() {bUsePreviewValueAsDefault=true, inputName=Material, inputType=FunctionInput_MaterialAttributes}
    B=MakeMaterialAttributes!(
     BaseColor=@n25
     Metallic
     Specular
     Roughness=@n26
     Anisotropy
     EmissiveColor=@n29
     Opacity
     OpacityMask
     Normal
     Tangent
     WorldPositionOffset
     SubsurfaceColor
     ClearCoat
     ClearCoatRoughness
     AmbientOcclusion
     Refraction
     CustomizedUVs_0
     CustomizedUVs_1
     CustomizedUVs_2
     CustomizedUVs_3
     CustomizedUVs_4
     CustomizedUVs_5
     CustomizedUVs_6
     CustomizedUVs_7
     PixelDepthOffset
     ShadingModel
     Displacement
    )
    Alpha=FunctionInput(Preview=@n30) {bUsePreviewValueAsDefault=true, inputName=Mask, inputType=FunctionInput_Scalar, sortPriority=1}
   ))
  MaterialFunctionCall!(NumberOfMips) {materialFunction=&/MF_LemonCart_CustomScreenMIPOut}
  n31 = Time()
  n32 = Sine(None=ConstantBiasScale(None=Sine(None=@n31) {period=1.33}))
  n33 = Sine(None=Add(A=@n31) {constB=0.7}) {period=0.66}
  n34 = Sine(None=Multiply(A=@n31) {constB=0.1}) {period=1.12}
  n35 = Sine(None=Add(A=ConstantBiasScale(None=@n33), B=ConstantBiasScale(None=@n34))) {period=0.91}
  Multiply(
   A=MaterialFunctionCall(X=ConstantBiasScale(None=@n32), Y=ConstantBiasScale(None=@n35)) {materialFunction=&MakeFloat2}.Result
   B=ScalarParameter() {defaultValue=3, parameterName=Tiling}
  ) {constB=5}
  n36 = MaterialFunctionCall!(
   UVs=NamedRerouteUsage()
   Speed
   Mask=MaterialFunctionCall!(Width=Constant() {r=0.35}, Sharpness=Constant() {r=0.9}, Offset, Direction Switch=StaticBool() {value=true}, Compare, Input Coordinates) {materialFunction=/Engine/Functions/Engine_MaterialFunctions03/Procedurals/GeneratedBand}.Result
  ) {materialFunction=&/MF_LemonCart_Subscreen_04}
  Multiply(
   A=VectorParameter() {parameterName=SubScreen04.Color, defaultValue=<rgba 1 1 1 1>}.RGB
   B=LinearInterpolate(B=@n36.Result, Alpha=@n36.Result) {constA=1}
  )
  n37 = Multiply(A=Time(), B=ScalarParameter() {defaultValue=3, parameterName=VerticalJitter.Speed}) {constB=0.1}
  n38 = Add(A=Add(A=Sine(None=@n37) {period=0.55}, B=Sine(None=@n37) {period=1.55}), B=Sine(None=@n37) {period=2.33})
  n39 = ScalarParameter() {defaultValue=0.003, parameterName=VerticalJitter.IntensityMax}
  n40 = NamedRerouteUsage()
  NamedRerouteDeclaration(None=Multiply(
    A=MaterialFunctionCall(Y=@n38) {materialFunction=&MakeFloat2}.Result
    B=LinearInterpolate(B=@n39, Alpha=@n40)
   ) {constB=0.003}) {name=VerticalJitter.FV2, nodeColor=<rgba 0.2891 0 1 1>}
  n41 = NamedRerouteUsage()
  n42 = Add(A=Time(), B=Divide(A=NamedRerouteUsage()))
  n43 = Multiply(
   A=Multiply(A=@n42, B=ScalarParameter() {defaultValue=0.3, parameterName=ScreenEffect.Speed}) {constB=0.1}
   B=ScalarParameter() {defaultValue=0.3, parameterName=ScreenEffect.RandomOffsetSpeed}
  )
  n44 = Sine(None=@n43) {period=0.55}
  n45 = Sine(None=@n43) {period=0.33}
  n46 = Add(A=Add(A=ConstantBiasScale(None=@n44), B=ConstantBiasScale(None=@n45)), B=ConstantBiasScale(None=Sine(None=@n43) {period=64}))
  n47 = Add(
   A=Multiply(A=@n42, B=ScalarParameter() {defaultValue=0.3, parameterName=ScreenEffect.Speed}) {constB=0.1}
   B=Multiply(A=Add(A=@n46) {constB=2}, B=ScalarParameter() {defaultValue=0.2, parameterName=ScreenEffect.OffsetIntensity}) {constB=0.2}
  )
  n48 = ComponentMask(None=Frac(None=ActorPositionWS())) {g=true, r=true}
  n49 = TextureSample!(UVs=Add(A=ComponentMask(None=@n41) {g=true}, B=Add(A=@n47, B=@n48)), Tex, Apply View MipBias) {samplerType=SAMPLERTYPE_LinearGrayscale, texture=/Game/Global/Textures/Default/Masks_and_Gradients/T_ScreenRamp_H}
  n50 = Time()
  n51 = Sine(None=@n50)
  n52 = Sine(None=@n50) {period=0.45}
  n53 = Saturate(None=Add(A=Add(A=@n51, B=@n52), B=Sine(None=@n50) {period=1.63}))
  NamedRerouteDeclaration(None=Multiply(A=Multiply(A=@n49.R, B=@n53), B=ScalarParameter() {defaultValue=0.2, parameterName=ScreenEffect.BumpOffsetIntensity}) {constB=0.2}) {name=ScanlineBumpOffsetHeight, nodeColor=<rgba 0.6875 1 0 1>}
  n54 = MaterialFunctionCall(Float3=ActorPositionWS()) {materialFunction=&BreakOutFloat3Components}
  n55 = Add(A=@n54.R, B=@n54.R)
  NamedRerouteDeclaration(None=Frac(None=Add(A=@n55, B=@n54.B))) {name=Seed, nodeColor=<rgba 0.3828 1 0 1>}
  n56 = MaterialFunctionCall(Float2=@n48) {materialFunction=&BreakOutFloat2Components}
  NamedRerouteDeclaration(None=Divide(A=Add(A=@n56.R, B=@n56.R))) {name=RandomAlpha, nodeColor=<rgba 1 0 0.5625 1>}
  NamedRerouteDeclaration(None=@n13) {name=CycleSpeed, nodeColor=<rgba 1 0 0.3516 1>}
  n57 = MaterialFunctionCall(
   UVs=NamedRerouteDeclaration(None=TextureCoordinate()) {name=TexCoord, nodeColor=<rgba 0.0547 1 0 1>}
   Texture Scale=MaterialFunctionCall(
    X=ScalarParameter() {defaultValue=1, group=ScreenControl, parameterName=SubScreens.UV.ScaleX}
    Y=ScalarParameter() {defaultValue=1, group=ScreenControl, parameterName=SubScreens.UV.ScaleY}
   ) {materialFunction=&MakeFloat2}.Result
  ) {materialFunction=&ScaleUVsByCenter}
  NamedRerouteDeclaration(None=@n57.UVs) {name=TexCoords.PostScale, nodeColor=<rgba 1 0.8438 0 1>}
  n58 = NamedRerouteUsage()
  n59 = MaterialFunctionCall(Float2=@n58) {materialFunction=&BreakOutFloat2Components}
  n60 = MaterialFunctionCall(Float2=OneMinus(None=@n58)) {materialFunction=&BreakOutFloat2Components}
  n61 = Multiply(A=Multiply(A=@n59.R, B=@n59.R), B=Multiply(A=@n60.R, B=@n60.R))
  n62 = OneMinus(None=Saturate(None=Multiply(A=@n61) {constB=40}))
  n63 = Round(None=Multiply(A=Power(Base=@n62) {constExponent=16}) {constB=2})
  n64 = ScalarParameter() {defaultValue=0.2, parameterName=EdgeOpacity}
  NamedRerouteDeclaration(None=Multiply(A=Saturate(None=@n63), B=@n64) {constB=0.1}) {name=EdgeMask, nodeColor=<rgba 0 1 0.1094 1>}
  n65 = Add(
   A=ComponentMask(None=NamedRerouteUsage()) {g=true}
   B=Multiply(A=Time(), B=ScalarParameter() {defaultValue=0.04, parameterName=Scanlines.AnimSpeed}) {constB=0.04}
  )
  n66 = ScalarParameter() {defaultValue=50, parameterName=Scanlines.Amount}
  n67 = ConstantBiasScale(None=Sine(None=Multiply(A=@n65, B=@n66) {constB=50}))
  n68 = ScalarParameter() {defaultValue=0.01, parameterName=Scanlines.Opacity}
  n69 = Power(Base=Multiply(A=@n49.R, B=LinearInterpolate(Alpha=@n53) {constA=0.4}))
  n70 = ScalarParameter() {defaultValue=1, parameterName=ScreenEffect.Opacity}
  NamedRerouteDeclaration(None=Add(A=Multiply(A=@n67, B=@n68) {constB=0.01}, B=Multiply(A=@n69, B=@n70) {constB=0.4})) {name=ScanlineOut, nodeColor=<rgba 0 1 0.3438 1>}
  n71 = Divide(A=@n63)
  NamedRerouteDeclaration(None=Multiply(A=Saturate(None=@n71), B=@n64) {constB=0.1}) {name=EdgeValues, nodeColor=<rgba 1 0.0234 0 1>}
  n72 = NamedRerouteUsage()
  n73 = Multiply(A=Add(A=Time(), B=Multiply(A=@n72) {constB=100}), B=ScalarParameter() {defaultValue=2, parameterName=Brightness.FlickerSpeed}) {constB=0.1}
  n74 = Sine(None=@n73) {period=0.57}
  n75 = Sine(None=@n73) {period=0.33}
  n76 = Add(A=Add(A=ConstantBiasScale(None=@n74), B=ConstantBiasScale(None=@n75)), B=ConstantBiasScale(None=Sine(None=@n73) {period=1.11}))
  NamedRerouteDeclaration(None=LinearInterpolate(A=ScalarParameter() {defaultValue=0.5, parameterName=Brightness.FlickerMin}, Alpha=Saturate(None=@n76)) {constA=0.5}) {name=BrightnessFlicker, nodeColor=<rgba 0 0.2969 1 1>}
  NamedRerouteDeclaration(None=@n57.0-1 mask) {name=TexCoords.PostScale.Mask, nodeColor=<rgba 1 0.8438 0 1>}
  n77 = MaterialFunctionCall(Radius=Constant() {r=0.7}, Density=Constant() {r=0.5}) {materialFunction=&RadialGradientExponential}
  n78 = ScalarParameter() {defaultValue=0.1, group=ScreenControl, parameterName=CenterScreenDiscoloration}
  NamedRerouteDeclaration(None=Add(
    A=Multiply(A=@n77.RadialGradientExponential, B=@n78) {constB=0.1}
    B=VectorParameter() {parameterName=Color.Base, defaultValue=<rgba 0.03 0.0285 0.0256 1>}.RGB
   )) {name=Color.Base, nodeColor=<rgba 0 1 0.9063 1>}
  TextureCoordinate()
 ]}
