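# M_LemonCart_Screen is a Default Lit material for the screens of the LemonCart environment set, built on material attributes.
# MF_LemonCartScreens supplies the screen material, whose result feeds the Material_In input of MF_FN_DestructibleDamage to layer in destructible damage.
# All shading logic lives in these two material functions, and automatic usage flagging in the editor is off.
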Material /Clyde/Environments/Sets/LemonCart/Materials/M_LemonCart_Screen
graph {
 properties {bAutomaticallySetUsageInEditor=false, bUseMaterialAttributes=true, usedShadingModels=Default Lit}
 nodes [MaterialFunctionCall(Material_In=MaterialFunctionCall!(Material, Mask) {materialFunction=/Clyde/Environments/Sets/LemonCart/MaterialFunctions/MF_LemonCartScreens}.Result) {materialFunction=&MF_FN_DestructibleDamage}]
}
